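# Translucent UI material that draws the rarity star row for a Battle Royale HUD quick bar slot.
# A star texture tiles through UV scaling, and a named-reroute mask built from offset U and Ceil cuts it to the NumStars share of TotalRarityTypes.
# IsSelected blends color from adaptive Dark to RarityColor, and lifts the masked star alpha from Unselected_Opacity to full.
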Material /BattleRoyaleGameUI/HUD/QuickBar/Assets/M_QuickBarSlot_RarityStars
graph {
 properties {bAutomaticallySetUsageInEditor=false, blendMode=BLEND_Translucent, materialDomain=MD_UI, usedShadingModels=Default Lit}
 nodes [
  Constant() {r=0.5}
  LinearInterpolate(
   A=MaterialFunctionCall!(BufferSource) {materialFunction=/BattleRoyaleGameUI/Shared/MF_UI_AdaptiveColors}.Dark
   B=VectorParameter() {parameterName=RarityColor, defaultValue=<rgba 1 1 1 1>}.RGB
   Alpha=ScalarParameter() {parameterName=IsSelected}
  )
  n1 = Constant()
  n2 = Add(A=MaterialFunctionCall() {materialFunction=&MF_UI_UVScaling}.ScaleY, B=AppendVector(A=@n1, B=@n1))
  n3 = Constant() {r=1}
  n4 = TextureSample(UVs=Frac(None=Fmod(A=@n2, B=@n3))) {samplerSource=SSM_Clamp_WorldGroupSettings, texture=/Game/UI/Foundation/Textures/Icons/Stars/T-Stars-Full-XS}
  n5 = ScalarParameter() {defaultValue=0.75, parameterName=Unselected_Opacity}
  Multiply(A=LinearInterpolate(A=Multiply(A=@n4.A, B=@n5), B=@n4.A, Alpha=ScalarParameter() {parameterName=IsSelected}), B=NamedRerouteUsage())
  n6 = Multiply(A=Divide(A=ScalarParameter() {defaultValue=7, parameterName=NumStars}, B=ScalarParameter() {defaultValue=7, parameterName=TotalRarityTypes}) {constB=6}) {constB=-1}
  n7 = Constant()
  n8 = ComponentMask(None=Add(A=TextureCoordinate(), B=AppendVector(A=@n6, B=@n7))) {r=true}
  NamedRerouteDeclaration(None=OneMinus(None=Ceil(None=@n8))) {name=RarityStarsMask, nodeColor=<rgba 0 0 0 1>}
 ]
}
