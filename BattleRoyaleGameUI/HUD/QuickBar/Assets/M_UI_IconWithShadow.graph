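# Translucent UI material for HUD quickbar icons that draws the icon over a soft drop shadow.
# Four offset samples of the Icon texture blend into a shadow mask, which combines with the icon's alpha at adjustable scale, rotation and opacity.
# An optional SpinningAnim switch adds an eased, time-driven half-turn with idle cycles between spins, plus animated shadow intensity.
# Static switches choose colors from the shared BR UI parameter collection and custom color parameters.
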Material /BattleRoyaleGameUI/HUD/QuickBar/Assets/M_UI_IconWithShadow
graph {
 properties {bAutomaticallySetUsageInEditor=false, blendMode=BLEND_Translucent, materialDomain=MD_UI, usedShadingModels=Default Lit}
 nodes [
  n1 = ScalarParameter() {defaultValue=0.05, parameterName=Shadow_X_Offset}
  n2 = Multiply(A=ScalarParameter() {defaultValue=0.05, parameterName=Shadow_Y_Offset}) {constB=-1}
  n3 = TextureSampleParameter2D(UVs=Add(A=NamedRerouteUsage(), B=AppendVector(A=@n1, B=@n2))) {parameterName=Icon, samplerSource=SSM_Clamp_WorldGroupSettings, texture=/Game/UI/Icons/Icon-QuickbarMode-Fight-M}
  n4 = NamedRerouteUsage()
  n5 = AppendVector(
   A=Multiply(A=ScalarParameter() {defaultValue=0.05, parameterName=Shadow_X_Offset}) {constB=-1}
   B=ScalarParameter() {defaultValue=0.05, parameterName=Shadow_Y_Offset}
  )
  n6 = AppendVector(
   A=ScalarParameter() {defaultValue=0.05, parameterName=Shadow_X_Offset}
   B=ScalarParameter() {defaultValue=0.05, parameterName=Shadow_Y_Offset}
  )
  n7 = Add(A=NamedRerouteUsage(), B=Multiply(A=@n6) {constB=-1})
  n8 = TextureCoordinate()
  n9 = Multiply(A=Time(), B=ScalarParameter() {defaultValue=1, parameterName=SpinAnimSpeed})
  n10 = MaterialFunctionCall("Value (0-1)"=Fmod(A=@n9, B=Constant() {r=1})) {materialFunction=&MF_UI_EaseCurves}
  n11 = MaterialFunctionCall!(A=Fmod(A=@n9, B=ScalarParameter() {defaultValue=3, parameterName=CyclesOff}), B=Constant() {r=1}) {materialFunction=/FortUILibrary/MaterialFunctions/Basics/MF_UI_BooleanOperations}
  n12 = LinearInterpolate(A=Multiply(A=OneMinus(None=@n10.EaseOutQuint)) {constB=0.5}, Alpha=@n11.IsGreater) {constB=0}
  n13 = ScalarParameter() {parameterName=Rotation}
  n14 = ScalarParameter() {defaultValue=0.8, parameterName=OverallIconScale}
  n15 = Subtract(A=@n14) {constB=0.125}
  n16 = NamedRerouteUsage()
  n17 = MaterialFunctionCall(
   UVs=StaticSwitchParameter(
    True=MaterialFunctionCall(UVs=@n8, "Rotation Angle (0-1)"=@n12) {materialFunction=&CustomRotator}.Rotated Values
    False=MaterialFunctionCall(UVs=@n8, "Rotation Angle (0-1)"=@n13) {materialFunction=&CustomRotator}.Rotated Values
   ) {parameterName=SpinningAnim}
   Texture Scale=LinearInterpolate(A=@n14, B=LinearInterpolate(A=@n14, B=@n15, Alpha=@n16), Alpha=ScalarParameter() {defaultValue=1, parameterName=UseScale})
  ) {materialFunction=&ScaleUVsByCenter}
  n18 = ScalarParameter() {defaultValue=1, parameterName=ShadowScale}
  n19 = Add(
   A=NamedRerouteDeclaration(None=MaterialFunctionCall(UVs=@n17.UVs, Texture Scale=@n18) {materialFunction=&ScaleUVsByCenter}.UVs) {name=UVs, nodeColor=<rgba 0 0 0 1>}
   B=@n6
  )
  n20 = LinearInterpolate(
   A=TextureSampleParameter2D(UVs=Add(A=@n4, B=@n5)) {parameterName=Icon, samplerSource=SSM_Clamp_WorldGroupSettings, texture=&Icon-QuickbarMode-Fight-M}.A
   Alpha=LinearInterpolate(
    A=TextureSampleParameter2D(UVs=@n7) {parameterName=Icon, samplerSource=SSM_Clamp_WorldGroupSettings, texture=&Icon-QuickbarMode-Fight-M}.A
    Alpha=TextureSampleParameter2D(UVs=@n19) {parameterName=Icon, samplerSource=SSM_Clamp_WorldGroupSettings, texture=&Icon-QuickbarMode-Fight-M}.A
   )
  )
  n21 = Multiply(
   A=NamedRerouteDeclaration(None=LinearInterpolate(A=@n3.A, Alpha=@n20)) {name=ShadowMask, nodeColor=<rgba 0 0 0 1>}
   B=ScalarParameter() {defaultValue=0.125, parameterName=ShadowOpacity}
  )
  n22 = Multiply(A=@n21, B=ScalarParameter() {defaultValue=1, parameterName=SpinAnimShadowIntensity}) {constB=2}
  n23 = NamedRerouteUsage()
  n24 = TextureSampleParameter2D(UVs=@n17.UVs) {parameterName=Icon, samplerSource=SSM_Clamp_WorldGroupSettings, texture=&Icon-QuickbarMode-Fight-M}
  LinearInterpolate(A=StaticSwitchParameter(True=LinearInterpolate(A=@n21, B=@n22, Alpha=@n23), False=@n21) {parameterName=SpinningAnim}, Alpha=@n24.A)
  NamedRerouteDeclaration(None=LinearInterpolate(A=OneMinus(None=@n10.EaseOutSine), Alpha=@n11.IsGreater) {constB=0}) {name=AnimCurve, nodeColor=<rgba 0 0 0 1>}
  n25 = ComponentMask(None=CollectionParameter() {collection=/BattleRoyaleGameUI/Shared/MPC_BRGameUI, parameterName=Yellow}) {b=true, g=true, r=true}
  n26 = VectorParameter() {parameterName=ClipEmptyColor, defaultValue=<rgba 0.1667 1 0 1>}
  n27 = NamedRerouteUsage()
  n28 = ComponentMask(None=CollectionParameter() {collection=&MPC_BRGameUI, parameterName=Charcoal}) {b=true, g=true, r=true}
  n29 = ComponentMask(None=CollectionParameter() {collection=&MPC_BRGameUI, parameterName=White}) {b=true, g=true, r=true}
  n30 = CollectionParameter() {collection=&MPC_BRGameUI, parameterName=White}
  StaticSwitchParameter(
   True=LinearInterpolate(
    A=VectorParameter() {parameterName=CustomOutlineColor, defaultValue=<rgba 0 0 0 1>}.RGB
    B=LinearInterpolate(A=@n25, B=@n26.RGB, Alpha=@n27)
    Alpha=@n24.A
   )
   False=StaticSwitchParameter(True=LinearInterpolate(A=@n28, B=@n29, Alpha=@n24.A), False=ComponentMask(None=@n30) {b=true, g=true, r=true}) {parameterName=HasAnim}
  ) {parameterName=MPCColor}
 ]
}
