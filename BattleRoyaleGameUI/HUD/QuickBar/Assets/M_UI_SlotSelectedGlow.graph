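# Translucent UI material that draws a soft glow for a selected quick bar slot in the battle royale HUD.
# A box SDF on scaled UVs is inverted and smoothstepped into a glow with adjustable size, corner radius and falloff.
# A sine ease-in over vertical UV fades the glow, which an Opacity parameter scales and IsSelected blends in.
# The graph also includes the shared adaptive colors function.
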Material /BattleRoyaleGameUI/HUD/QuickBar/Assets/M_UI_SlotSelectedGlow
graph {
 properties {bAutomaticallySetUsageInEditor=false, blendMode=BLEND_Translucent, materialDomain=MD_UI, usedShadingModels=Default Lit}
 nodes [
  n1 = MaterialFunctionCall() {materialFunction=&MF_UI_UVScaling}
  n2 = ScalarParameter() {defaultValue=0.15, parameterName=Size}
  n3 = ScalarParameter() {parameterName=GlowCornerRadius}
  n4 = StaticBool()
  n5 = SmoothStep(
   Min=ScalarParameter() {defaultValue=0.2, parameterName=Min}
   Max=ScalarParameter() {defaultValue=1, parameterName=Max}
   Value=OneMinus(None=MaterialFunctionCall(UVs=@n1.ScaleX, Size=@n2, CornerRoundness=@n3, KeepAspectRatio=@n4) {materialFunction=&MF_UI_Box}.SDF)
  )
  n6 = MaterialFunctionCall() {materialFunction=&GetUserInterfaceUV}
  n7 = Saturate(None=MaterialFunctionCall("Value (0-1)"=ComponentMask(None=@n6.Normalized UV) {g=true}) {materialFunction=&MF_UI_EaseCurves}.EaseInSine)
  LinearInterpolate(
   B=Multiply(A=Multiply(A=@n5, B=@n7), B=ScalarParameter() {defaultValue=0.5, parameterName=Opacity})
   Alpha=ScalarParameter() {parameterName=IsSelected}
  )
  MaterialFunctionCall!(BufferSource) {materialFunction=/BattleRoyaleGameUI/Shared/MF_UI_AdaptiveColors}
 ]
}
