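# UI material function for Battle Royale game UI that produces an animated rainbow gradient as Main Color and Accent Color outputs.
# Fit-scaled, rotated UVs feed a slowly panning, tiled and wrapped linear gradient, plus a separately panning, low-res blurred noise texture that breaks up the bands.
# Their sum drives HueShift on the input color, yielding Main Color; a Color Burn blend with Accent Intensity derives Accent Color.
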MaterialFunction /BattleRoyaleGameUI/Shared/MF_UI_BRGameUI_RainbowStyling
graph {nodes [
  n1 = FunctionInput() {bUsePreviewValueAsDefault=true, inputName=Gradient Rotation, inputType=FunctionInput_Scalar, sortPriority=1, previewValue=<0.2 0 0 1>}
  n2 = MaterialFunctionCall(
   UVs=MaterialFunctionCall() {materialFunction=&MF_UI_UVScaling}.Scaled to Fit
   "Rotation Angle (0-1)"=@n1
  ) {materialFunction=&CustomRotator}
  n3 = FunctionInput() {bUsePreviewValueAsDefault=true, inputName=Gradient Tiling, inputType=FunctionInput_Scalar, sortPriority=2, previewValue=<2 0 0 1>}
  n4 = Multiply(
   A=MaterialFunctionCall(UV Channel=Panner(Coordinate=@n2.Rotated Values) {speedX=0.05}) {materialFunction=&LinearGradient}.UGradient
   B=@n3
  ) {constB=2}
  n5 = FunctionInput() {bUsePreviewValueAsDefault=true, inputName=Noise Size, inputType=FunctionInput_Scalar, sortPriority=3, previewValue=<0.5 0 0 1>}
  n6 = TextureSample(UVs=Panner(Coordinate=Multiply(A=@n2.Rotated Values, B=@n5) {constB=0.5}) {speedX=0.1, speedY=0.05}) {samplerType=SAMPLERTYPE_Grayscale, texture=/Game/TimeOfDay/TODM_Sequence/Materials/Sky/Textures/LowResBlurredNoise_Radish_LF_Root}
  n7 = FunctionInput() {bUsePreviewValueAsDefault=true, inputName=Noise Opacity, inputType=FunctionInput_Scalar, sortPriority=4, previewValue=<0.2 0 0 1>}
  n8 = FunctionInput() {bUsePreviewValueAsDefault=true, inputName=Hue Color, sortPriority=5, previewValue=<0.5394 1 0.5386 1>}
  n9 = MaterialFunctionCall(Hue Shift Percentage=Add(A=Frac(None=@n4), B=Multiply(A=@n6.RGB, B=@n7) {constB=0.2}), Texture=@n8) {materialFunction=&HueShift}
  FunctionOutput(None=@n9.Result) {outputName=Main Color}
  n10 = FunctionInput() {bUsePreviewValueAsDefault=true, inputName=Accent Intensity, inputType=FunctionInput_Scalar, sortPriority=6, previewValue=<0.65 0 0 1>}
  FunctionOutput(None=MaterialFunctionCall!(Base=@n9.Result, Blend=@n10) {materialFunction=/Engine/Functions/Engine_MaterialFunctions03/Blends/Blend_ColorBurn}.Result) {outputName=Accent Color, sortPriority=2}
 ]}
